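# Material function for the GoldenLychee vehicle wheel that layers animated, speed-reactive stripe effects onto MF_WheelParent_Simple.
# Two panning samples of the speed-stripe texture on UV1 are scaled by normalized ObjectSpeed: red fades with speed, inverted green strengthens.
# The tinted red layer drives Emissive, and both layers blend it into Base Color within the vertex-color red mask.
# A DebugDisplayFX static switch bypasses the speed scaling.
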MaterialFunction /VehicleCosmetics/Wheels/GoldenLychee/Materials/MF_Wheel_GoldenLychee
graph {nodes [
  n1 = MaterialFunctionCall() {materialFunction=&MF_WheelParent_Simple}
  n2 = Time()
  n3 = TextureSample(UVs=Panner(Coordinate=TextureCoordinate() {coordinateIndex=1, uTiling=2}, Time=@n2) {speedX=1.25}) {texture=/VehicleCosmetics/Wheels/GoldenLychee/Textures/T_GoldenLychee_Speedstripes}
  n4 = Saturate(None=Divide(A=ScalarParameter() {parameterName=ObjectSpeed}, B=ScalarParameter() {defaultValue=1000, group=0_Animated, parameterName=SpeedDivider}) {constA=0})
  n5 = OneMinus(None=@n4)
  n6 = Multiply(A=@n3.R, B=Add(B=Clamp(None=@n5) {minDefault=0.1}) {constA=0.2})
  n7 = Saturate(None=StaticSwitchParameter(True=@n3.R, False=Subtract(A=@n6) {constB=0.2}) {group=0_Animated, parameterName=DebugDisplayFX})
  n8 = LinearInterpolate(B=Multiply(A=VectorParameter() {group=0_Animated, parameterName=Color, defaultValue=<rgba 0.1166 0.4535 0.4878 1>}.RGB, B=@n7), Alpha=@n7) {constB=0}
  n9 = GetMaterialAttributes(None=@n1.MaterialAttributes) {attributeGetTypes [#1, #2]}
  n10 = TextureCoordinate() {coordinateIndex=1, uTiling=3, vTiling=2}
  n11 = OneMinus(None=TextureSample(UVs=Panner(Coordinate=@n10, Time=@n2) {speedX=0.75, speedY=0.2}) {texture=&T_GoldenLychee_Speedstripes}.G)
  n12 = StaticSwitchParameter(True=@n11, False=Subtract(A=Multiply(A=@n11, B=@n4)) {constB=0.1}) {parameterName=DebugDisplayFX}
  n13 = LinearInterpolate(A=@n9.BaseColor, B=@n8, Alpha=Add(A=@n7, B=Saturate(None=@n12))) {constB=0}
  n14 = VertexColor()
  FunctionOutput(None=SetMaterialAttributes!(MaterialAttributes=@n1.MaterialAttributes, EmissiveColor=@n8, BaseColor=LinearInterpolate(A=@n9.BaseColor, B=@n13, Alpha=@n14.R) {constB=0}) {attributeSetTypes [#2, #1]})
 ]}
